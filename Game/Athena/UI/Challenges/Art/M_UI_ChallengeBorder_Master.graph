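# Translucent UI master material for challenge borders, with a parameterized color gradient and an optional bolt-shaped edge mask.
# Rotated, scaled UVs blend primary and secondary colors, and AnimateHighlight fades to a hover look.
# The IsRewardChallenge switch replaces hover colors with a dark gradient mixing subtract and multiply darkening.
# The procedural, invertible bolt mask joins angled left and right edges across a center split and scales alpha when UseBoltMask is on.
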Material /Game/Athena/UI/Challenges/Art/M_UI_ChallengeBorder_Master
graph {
 properties {
  bUsedWithStaticMesh=false
  blendMode=BLEND_Translucent
  materialDomain=MD_UI
  pixelDepthOffsetMode=PDOM_Legacy
  usedShadingModels=Default Lit
 }
 nodes [
  n1 = TextureCoordinate() {coordinateIndex=4}
  n2 = Rotator(
   Coordinate=Add(A=@n1) {constB=-0.5}
   Time=Divide(A=ScalarParameter() {defaultValue=5, group=BoltEdge, parameterName=CenterAngle}) {constB=360}
  ) {speed=6.2832}
  n3 = ScalarParameter() {group=BoltEdge, parameterName=CenterOffset}
  n4 = ScalarParameter() {group=BoltEdge, parameterName=CenterOffset_Highlight}
  n5 = Add(A=@n2, B=Multiply(A=Add(A=@n3, B=@n4)) {constB=-1})
  n6 = ScalarParameter() {defaultValue=32, group=BoltEdge, parameterName=CenterEdgeSharpness}
  Clamp(None=Multiply(A=ComponentMask(None=@n5) {r=true}, B=@n6))
  n7 = VectorParameter() {group=ColorGradient, parameterName=Secondary, defaultValue=<rgba 0.7011 0.2346 0.5089 1>}
  n8 = VectorParameter() {group=ColorGradient, parameterName=Primary, defaultValue=<rgba 0.2086 0.0086 0.0382 1>}
  n9 = ScalarParameter() {defaultValue=4, group=ColorGradient, parameterName=GradientPosition_Y}
  n10 = ScalarParameter() {group=ColorGradient, parameterName=GradientPosition_X}
  n11 = TextureCoordinate() {coordinateIndex=4}
  n12 = Divide(A=Add(A=@n11, B=AppendVector(A=@n9, B=@n10)), B=ScalarParameter() {defaultValue=1.5, group=ColorGradient, parameterName=GradientScale})
  n13 = ScalarParameter() {defaultValue=8, group=ColorGradient, parameterName=GradientRotation}
  n14 = Divide(A=@n13) {constB=360}
  n15 = Clamp(None=ComponentMask(None=Rotator(Coordinate=@n12, Time=@n14) {centerX=1, speed=6.2832}) {g=true})
  n16 = LinearInterpolate(A=@n7.RGB, B=@n8.RGB, Alpha=@n15)
  n17 = ScalarParameter() {defaultValue=4, group=ColorGradient, parameterName=DarkGradientPosition_Y}
  n18 = ScalarParameter() {group=ColorGradient, parameterName=DarkGradientPosition_X}
  n19 = Divide(A=Add(A=@n11, B=AppendVector(A=@n17, B=@n18)), B=ScalarParameter() {defaultValue=1.5, group=ColorGradient, parameterName=DarkGradientScale})
  n20 = Divide(A=@n13) {constB=360}
  n21 = Clamp(None=ComponentMask(None=Rotator(Coordinate=@n19, Time=@n20) {centerX=1, speed=6.2832}) {g=true})
  n22 = Multiply(A=@n21) {constB=0.125}
  n23 = Subtract(A=@n16, B=@n22)
  n24 = OneMinus(None=@n21)
  n25 = Multiply(A=@n16, B=@n24)
  n26 = ScalarParameter() {defaultValue=0.5, group=ColorGradient, parameterName=MultBurnBlend}
  n27 = ScalarParameter() {group=HighlightGradient, parameterName=DarkGradientFill_Highlight}
  n28 = Subtract(A=@n16, B=Add(A=@n22, B=@n27))
  n29 = Multiply(A=@n16, B=Multiply(A=@n24, B=OneMinus(None=@n27)))
  n30 = VectorParameter() {group=HighlightGradient, parameterName=HoverSecondary, defaultValue=<rgba 0.7011 0.4017 0.5777 1>}
  n31 = VectorParameter() {group=HighlightGradient, parameterName=HoverPrimary, defaultValue=<rgba 0.8125 0.0052 0.4024 1>}
  LinearInterpolate(
   A=StaticSwitchParameter(True=LinearInterpolate(A=@n23, B=@n25, Alpha=@n26), False=@n16) {parameterName=IsRewardChallenge}
   B=StaticSwitchParameter(True=LinearInterpolate(A=@n28, B=@n29, Alpha=@n26), False=LinearInterpolate(A=@n30.RGB, B=@n31.RGB, Alpha=@n15)) {parameterName=IsRewardChallenge}
   Alpha=ScalarParameter() {group=HighlightGradient, parameterName=AnimateHighlight, sliderMax=1}
  )
  n32 = LinearInterpolate(A=@n7.A, B=@n8.A, Alpha=@n15)
  n33 = LinearInterpolate(A=@n30.A, B=@n31.A, Alpha=@n15)
  n34 = ScalarParameter() {group=HighlightGradient, parameterName=AnimateHighlight, sliderMax=1}
  n35 = Rotator(Coordinate=@n1, Time=Divide(A=ScalarParameter() {defaultValue=10, group=BoltEdge, parameterName=LeftAngle}) {constB=360}) {centerY=0, speed=6.2832}
  n36 = ScalarParameter() {defaultValue=0.9081, group=BoltEdge, parameterName=LeftOffset}
  n37 = Add(A=@n35, B=Add(A=@n36) {constB=-1})
  n38 = ScalarParameter() {defaultValue=128, group=BoltEdge, parameterName=LeftRightEdgeSharpness}
  n39 = Rotator(Coordinate=@n1, Time=Divide(A=ScalarParameter() {defaultValue=10, group=BoltEdge, parameterName=RightAngle}) {constB=360}) {centerY=1, speed=6.2832}
  n40 = ScalarParameter() {defaultValue=0.8823, group=BoltEdge, parameterName=RightOffset}
  n41 = Add(A=@n39, B=Add(A=@n40) {constB=-1})
  n42 = ComponentMask(None=Add(A=@n2, B=Multiply(A=@n3) {constB=-1})) {r=true}
  n43 = LinearInterpolate(
   A=Multiply(A=ComponentMask(None=@n37) {g=true}, B=@n38)
   B=Multiply(A=ComponentMask(None=@n41) {g=true}, B=@n38)
   Alpha=Clamp(None=Multiply(A=@n42, B=@n6))
  )
  n44 = OneMinus(None=@n43)
  n45 = ScalarParameter() {group=Switches, parameterName=InverseBoltMask}
  LinearInterpolate(
   B=StaticSwitchParameter(True=Constant() {r=1}, False=LinearInterpolate(A=@n32, B=@n33, Alpha=@n34)) {parameterName=IsRewardChallenge}
   Alpha=StaticSwitchParameter(True=LinearInterpolate(A=@n43, B=@n44, Alpha=@n45), False=Constant() {r=1}) {group=Switches, parameterName=UseBoltMask}
  )
  n46 = Add(A=Add(A=@n36, B=ScalarParameter() {defaultValue=-0.05, group=BoltEdge, parameterName=LeftOffset_Highlight})) {constB=-1}
  n47 = ScalarParameter() {defaultValue=128, group=BoltEdge, parameterName=LeftRightHighlightSharpness}
  Multiply(A=ComponentMask(None=Add(A=@n35, B=@n46)) {g=true}, B=@n47)
  n48 = Add(A=Add(A=@n40, B=ScalarParameter() {defaultValue=-0.05, group=BoltEdge, parameterName=RightOffset_Highlight})) {constB=-1}
  Multiply(A=ComponentMask(None=Add(A=@n39, B=@n48)) {g=true}, B=@n47)
 ]
}
