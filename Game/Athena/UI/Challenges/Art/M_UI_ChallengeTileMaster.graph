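# Translucent UI master material for challenge tiles, with static switches for locked state, bolt-shaped cut, spikes and color treatments.
# Rotated, scaled UV gradients blend Primary, Secondary and optional Accent colors, plus an optional burned dark gradient.
# Procedural border and zigzag bolt-edge masks, plus a spike texture, reveal the lighter fill and add a color-dodge highlight.
# Opacity follows the invertible bolt mask, or gradient alpha plus border when locked.
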Material /Game/Athena/UI/Challenges/Art/M_UI_ChallengeTileMaster
graph {
 properties {
  bUsedWithStaticMesh=false
  blendMode=BLEND_Translucent
  materialDomain=MD_UI
  pixelDepthOffsetMode=PDOM_Legacy
  usedShadingModels=Default Lit
 }
 nodes [
  n1 = VectorParameter() {group=ColorGradient, parameterName=Secondary, defaultValue=<rgba 0.7011 0.2346 0.5089 1>}
  n2 = VectorParameter() {group=ColorGradient, parameterName=Primary, defaultValue=<rgba 0.2086 0.0086 0.0382 1>}
  n3 = ScalarParameter() {defaultValue=4.25, group=ColorGradient, parameterName=GradientPosition_Y}
  n4 = ScalarParameter() {group=ColorGradient, parameterName=GradientPosition_X}
  n5 = TextureCoordinate() {coordinateIndex=4}
  n6 = Divide(A=Add(A=@n5, B=AppendVector(A=@n3, B=@n4)), B=ScalarParameter() {defaultValue=1.5, group=ColorGradient, parameterName=GradientScale})
  n7 = ScalarParameter() {defaultValue=8, group=ColorGradient, parameterName=GradientRotation}
  n8 = Divide(A=@n7) {constB=360}
  n9 = Clamp(None=ComponentMask(None=Rotator(Coordinate=@n6, Time=@n8) {speed=6.2832}) {g=true})
  n10 = ScalarParameter() {defaultValue=0.5, group=Locked, parameterName=BorderAlpha}
  n11 = Rotator(Coordinate=@n5, Time=Divide(A=ScalarParameter() {defaultValue=-0.15, group=BorderHighlight, parameterName=Rotation_X}) {constB=360}) {speed=6.2832}
  n12 = ComponentMask(None=Abs(None=Add(A=@n11) {constB=-0.5})) {r=true}
  n13 = ScalarParameter() {defaultValue=0.45, group=BorderHighlight, parameterName=Offset_X}
  n14 = Add(A=@n13, B=ScalarParameter() {defaultValue=0.01, group=BorderHighlight, parameterName=SSMargin_X, sliderMax=1})
  n15 = Rotator(Coordinate=@n5, Time=Divide(A=ScalarParameter() {defaultValue=0.125, group=BorderHighlight, parameterName=Rotation_Y}) {constB=360}) {speed=6.2832}
  n16 = ComponentMask(None=Abs(None=Add(A=@n15) {constB=-0.5})) {g=true}
  n17 = ScalarParameter() {defaultValue=0.4625, group=BorderHighlight, parameterName=Offset_Y}
  n18 = Add(A=@n17, B=ScalarParameter() {defaultValue=0.005, group=BorderHighlight, parameterName=SSMargin_Y, sliderMax=1})
  n19 = TextureCoordinate() {coordinateIndex=4}
  n20 = Rotator(Coordinate=@n19, Time=Divide(A=ScalarParameter() {defaultValue=10, group=BoltEdge, parameterName=LeftAngle}) {constB=360}) {centerY=0, speed=6.2832}
  n21 = ScalarParameter() {defaultValue=0.9081, group=BoltEdge, parameterName=LeftOffset}
  n22 = Add(A=@n21, B=ScalarParameter() {defaultValue=-0.05, group=BoltEdge, parameterName=LeftOffset_Highlight})
  n23 = ComponentMask(None=Add(A=@n20, B=Add(A=@n22) {constB=-1})) {g=true}
  n24 = ScalarParameter() {defaultValue=128, group=BoltEdge, parameterName=LeftRightHighlightSharpness}
  n25 = Rotator(Coordinate=@n19, Time=Divide(A=ScalarParameter() {defaultValue=10, group=BoltEdge, parameterName=RightAngle}) {constB=360}) {centerY=1, speed=6.2832}
  n26 = ScalarParameter() {defaultValue=0.8823, group=BoltEdge, parameterName=RightOffset}
  n27 = Add(A=@n26, B=ScalarParameter() {defaultValue=-0.05, group=BoltEdge, parameterName=RightOffset_Highlight})
  n28 = ComponentMask(None=Add(A=@n25, B=Add(A=@n27) {constB=-1})) {g=true}
  n29 = Rotator(
   Coordinate=Add(A=@n19) {constB=-0.5}
   Time=Divide(A=ScalarParameter() {defaultValue=5, group=BoltEdge, parameterName=CenterAngle}) {constB=360}
  ) {speed=6.2832}
  n30 = ScalarParameter() {group=BoltEdge, parameterName=CenterOffset}
  n31 = Multiply(A=Add(A=@n30, B=ScalarParameter() {group=BoltEdge, parameterName=CenterOffset_Highlight})) {constB=-1}
  n32 = ScalarParameter() {defaultValue=32, group=BoltEdge, parameterName=CenterEdgeSharpness}
  n33 = Multiply(A=ComponentMask(None=Add(A=@n29, B=@n31)) {r=true}, B=@n32)
  n34 = OneMinus(None=LinearInterpolate(A=Multiply(A=@n23, B=@n24), B=Multiply(A=@n28, B=@n24), Alpha=Clamp(None=@n33)))
  n35 = Add(
   A=LinearInterpolate(
    A=MaterialFunctionCall(Alpha=@n12, Min=@n13, Max=@n14) {materialFunction=&SmoothStep}.Result
    Alpha=MaterialFunctionCall(Alpha=@n16, Min=@n17, Max=@n18) {materialFunction=&SmoothStep}.Result
   )
   B=StaticSwitchParameter(True=Clamp(None=@n34), False=Constant()) {group=Switches, parameterName=UseBoltMask}
  )
  n36 = Clamp(None=@n35)
  n37 = Add(A=@n20, B=Add(A=@n21) {constB=-1})
  n38 = ScalarParameter() {defaultValue=128, group=BoltEdge, parameterName=LeftRightEdgeSharpness}
  n39 = Add(A=@n25, B=Add(A=@n26) {constB=-1})
  n40 = ComponentMask(None=Add(A=@n29, B=Multiply(A=@n30) {constB=-1})) {r=true}
  n41 = LinearInterpolate(
   A=Multiply(A=ComponentMask(None=@n37) {g=true}, B=@n38)
   B=Multiply(A=ComponentMask(None=@n39) {g=true}, B=@n38)
   Alpha=Clamp(None=Multiply(A=@n40, B=@n32))
  )
  n42 = OneMinus(None=@n41)
  n43 = ScalarParameter() {group=Switches, parameterName=InverseBoltMask}
  StaticSwitchParameter(
   True=Add(A=LinearInterpolate(A=@n1.A, B=@n2.A, Alpha=@n9), B=LinearInterpolate(B=@n10, Alpha=@n36))
   False=StaticSwitchParameter(True=LinearInterpolate(A=@n41, B=@n42, Alpha=@n43), False=Constant() {r=1}) {group=Switches, parameterName=UseBoltMask}
  ) {group=Switches, parameterName=IsLocked}
  n44 = LinearInterpolate(A=@n1.RGB, B=@n2.RGB, Alpha=@n9)
  n45 = ScalarParameter() {defaultValue=4, group=ColorGradient, parameterName=DarkGradientPosition_Y}
  n46 = ScalarParameter() {group=ColorGradient, parameterName=DarkGradientPosition_X}
  n47 = Divide(A=Add(A=@n5, B=AppendVector(A=@n45, B=@n46)), B=ScalarParameter() {defaultValue=1.5, group=ColorGradient, parameterName=DarkGradientScale})
  n48 = Divide(A=StaticSwitchParameter(True=ScalarParameter() {defaultValue=8, group=ColorGradient, parameterName=DarkGradientRotation}, False=@n7) {group=Switches, parameterName=SplitDarkGradientRotation}) {constB=360}
  n49 = Clamp(None=ComponentMask(None=Rotator(Coordinate=@n47, Time=@n48) {speed=6.2832}) {g=true})
  n50 = Multiply(A=@n49) {constB=0.125}
  n51 = OneMinus(None=@n49)
  n52 = VectorParameter() {group=ColorGradient, parameterName=Accent, defaultValue=<rgba 1 0.007 0.1683 1>}
  n53 = LinearInterpolate(A=@n52.RGB, B=Constant() {r=1}, Alpha=ScalarParameter() {defaultValue=0.5, group=ColorGradient, parameterName=WhiteFillAmount})
  n54 = StaticSwitchParameter(
   True=LinearInterpolate(A=LinearInterpolate(A=@n53, B=@n52.RGB, Alpha=@n9), B=@n44, Alpha=ScalarParameter() {group=ColorGradient, parameterName=BaseFillAmount})
   False=@n44
  ) {group=Switches, parameterName=UseRGB3}
  n55 = TextureCoordinate()
  n56 = OneMinus(None=ComponentMask(None=@n55) {r=true})
  n57 = ComponentMask(None=@n55) {g=true}
  n58 = TextureSample(UVs=StaticSwitchParameter(True=AppendVector(A=@n56, B=@n57), False=@n55) {group=Switches, parameterName=FlipUSpikes}) {samplerType=SAMPLERTYPE_LinearColor, texture=&/T_UI_DF_ChallengeTileSpikes}
  n59 = Multiply(A=ScalarParameter() {defaultValue=0.2, group=Spikes, parameterName=SSMarginSpikes}) {constB=0.5}
  n60 = Multiply(A=@n59) {constB=-1}
  n61 = ScalarParameter() {defaultValue=0.5, group=Spikes, parameterName=SpikeScale}
  n62 = Add(
   A=@n35
   B=MaterialFunctionCall(
    Alpha=StaticComponentMaskParameter(None=@n58.RGB) {defaultR=true, group=Spikes, parameterName=SpikesMask}
    Min=Add(A=@n60, B=@n61)
    Max=Add(A=@n59, B=@n61)
   ) {materialFunction=&SmoothStep}.Result
  )
  n63 = Constant()
  n64 = StaticSwitchParameter(True=@n62, False=@n63) {defaultValue=true, group=Switches, parameterName=UseBorderHighlight}
  n65 = StaticSwitchParameter(True=@n35, False=@n63) {defaultValue=true, group=Switches, parameterName=UseBorderHighlight}
  n66 = LinearInterpolate(
   A=LinearInterpolate(
    A=Subtract(A=@n44, B=@n50)
    B=Multiply(A=@n44, B=@n51)
    Alpha=ScalarParameter() {defaultValue=0.5, group=ColorGradient, parameterName=MultBurnBlend}
   )
   B=@n54
   Alpha=Clamp(None=StaticSwitchParameter(True=@n64, False=@n65) {defaultValue=true, group=Switches, parameterName=UseSpikes})
  )
  n67 = Constant()
  n68 = ScalarParameter() {defaultValue=0.2, group=BorderHighlight, parameterName=HighlightAlpha}
  n69 = Constant()
  n70 = StaticSwitchParameter(
   True=StaticSwitchParameter(True=Clamp(None=@n62, Min=@n67, Max=@n68), False=@n69) {defaultValue=true, group=Switches, parameterName=UseBorderHighlight}
   False=StaticSwitchParameter(True=Clamp(None=@n35, Min=@n67, Max=@n68), False=@n69) {defaultValue=true, group=Switches, parameterName=UseBorderHighlight}
  ) {defaultValue=true, group=Switches, parameterName=UseSpikes}
  StaticSwitchParameter(
   True=LinearInterpolate(A=VectorParameter() {parameterName=RGB_LockedFill, defaultValue=<rgba 0 0.2159 1 0>}.RGB, B=@n44, Alpha=@n36)
   False=Divide(A=StaticSwitchParameter(True=@n66, False=@n54) {defaultValue=true, group=Switches, parameterName=UseDarkGradient}, B=OneMinus(None=@n70))
  ) {group=Switches, parameterName=IsLocked}
 ]
}
