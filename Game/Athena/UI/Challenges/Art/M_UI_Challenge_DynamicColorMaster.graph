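# Translucent UI master material that outputs a dynamic flat color for Athena challenge UI.
# Static switches pick the Accent, Secondary or Primary color parameter, with the chosen color's alpha scaled by an Alpha scalar.
# An optional color burn darkens the RGB by blending subtract and multiply variants driven by BurnFill.
# An optional WhiteFill switch then lerps the result toward white.
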Material /Game/Athena/UI/Challenges/Art/M_UI_Challenge_DynamicColorMaster
graph {
 properties {
  bUsedWithStaticMesh=false
  blendMode=BLEND_Translucent
  materialDomain=MD_UI
  pixelDepthOffsetMode=PDOM_Legacy
  usedShadingModels=Default Lit
 }
 nodes [
  n1 = VectorParameter() {parameterName=Accent, defaultValue=<rgba 1 0.007 0.1683 1>}
  n2 = VectorParameter() {parameterName=Secondary, defaultValue=<rgba 0.7011 0.2346 0.5089 1>}
  n3 = VectorParameter() {parameterName=Primary, defaultValue=<rgba 0.2086 0.0086 0.0382 1>}
  Multiply(
   A=StaticSwitchParameter(True=@n1.A, False=StaticSwitchParameter(True=@n2.A, False=@n3.A) {parameterName=UseSecondary}) {parameterName=UseAccent}
   B=ScalarParameter() {defaultValue=1, parameterName=Alpha}
  )
  n4 = StaticSwitchParameter(True=@n1.RGB, False=StaticSwitchParameter(True=@n2.RGB, False=@n3.RGB) {parameterName=UseSecondary}) {parameterName=UseAccent}
  n5 = ScalarParameter() {defaultValue=0.1875, parameterName=BurnFill}
  n6 = Multiply(A=@n5) {constB=0.125}
  n7 = OneMinus(None=@n5)
  n8 = StaticSwitchParameter(
   True=LinearInterpolate(
    A=Subtract(A=@n4, B=@n6)
    B=Multiply(A=@n4, B=@n7)
    Alpha=ScalarParameter() {defaultValue=0.5, group=ColorGradient, parameterName=MultBurnBlend}
   )
   False=@n4
  ) {parameterName=UseColorBurn}
  StaticSwitchParameter(True=LinearInterpolate(A=@n8, Alpha=ScalarParameter() {parameterName=WhiteFill}), False=@n8) {parameterName=UseWhiteFill}
 ]
}
