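# Translucent UI material for the datachip challenges screen, drawing an animated circuit pattern on screen-aligned, panning, rotated, tiled UVs.
# Texture channels supply nodes (R), traces (G) and a gradient (B) that a sine-driven band sweeps along the traces.
# Two or optionally three panning noise layers are multiplied to reveal the circuits in patches.
# A static switch adds a burst: animated radial gradient plus rippling circuit reveal, with optional center chip cutout.
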Material /Game/Athena/UI/Challenges/Art/M_UI_Datachips_CircuitPattern
graph {
 properties {
  bUsedWithStaticMesh=false
  blendMode=BLEND_Translucent
  materialDomain=MD_UI
  pixelDepthOffsetMode=PDOM_Legacy
  usedShadingModels=Default Lit
 }
 nodes [
  n1 = MaterialFunctionCall(Texture Dimensions=TextureCoordinate()) {materialFunction=&ScreenAlignedUVs}
  n2 = ScalarParameter() {group=CircuitUVs, parameterName=Circuit_OffsetU}
  n3 = ScalarParameter() {group=CircuitUVs, parameterName=Circuit_OffsetV}
  n4 = Panner(
   Coordinate=Add(A=@n1."X 100%, Y 100%", B=AppendVector(A=@n2, B=@n3))
   Speed=AppendVector(
    A=ScalarParameter() {defaultValue=-0.0025, group=CircuitUVs, parameterName=Circuit_SpeedU}
    B=ScalarParameter() {defaultValue=-0.005, group=CircuitUVs, parameterName=Circuit_SpeedV}
   )
  ) {speedX=0.25}
  n5 = ScalarParameter() {defaultValue=0.05, group=CircuitUVs, parameterName=Circuit_Rotation}
  n6 = ScalarParameter() {defaultValue=6, group=CircuitUVs, parameterName=Circuit_TileU}
  n7 = ScalarParameter() {defaultValue=6, group=CircuitUVs, parameterName=Circuit_TileV}
  n8 = Frac(None=Multiply(A=Rotator(Coordinate=@n4, Time=@n5), B=AppendVector(A=@n6, B=@n7)))
  n9 = TextureSampleParameter2D(UVs=@n8) {parameterName=CircuitTexture, texture=&/T_UI_Datachips_Circuits}
  n10 = Add(A=@n9.RGB, B=@n9.RGB)
  n11 = Add(A=@n10, B=@n10)
  Add(A=@n11, B=@n11)
  n12 = ScalarParameter() {group=BurstAnim, parameterName=AnimateBurst}
  n13 = MaterialFunctionCall(
   CenterPosition=AppendVector(
    A=ScalarParameter() {defaultValue=0.5, group=BurstAnim, parameterName=BurstPositionU}
    B=ScalarParameter() {defaultValue=0.5, group=BurstAnim, parameterName=BurstPositionV}
   )
   Radius=LinearInterpolate(A=ScalarParameter() {group=BurstAnim, parameterName=RadiusStart}, B=ScalarParameter() {group=BurstAnim, parameterName=RadiusEnd}, Alpha=@n12)
   Density=LinearInterpolate(
    A=ScalarParameter() {group=BurstAnim, parameterName=DensityStart}
    B=ScalarParameter() {group=BurstAnim, parameterName=DensityEnd}
    Alpha=@n12
   )
  ) {materialFunction=&RadialGradientExponential}
  n14 = Multiply(A=@n13.RadialGradientExponential, B=ScalarParameter() {group=BurstAnim, parameterName=RadialGradMult})
  n15 = Abs(None=Add(A=TextureCoordinate(), B=Constant() {r=-0.5}))
  n16 = ScalarParameter() {defaultValue=0.125, group=BurstAnim, parameterName=ChipMaskThicknessv}
  n17 = ScalarParameter() {defaultValue=0.0012, group=BurstAnim, parameterName=SSMarginChipMask}
  n18 = Multiply(A=@n17) {constB=-1}
  n19 = OneMinus(None=MaterialFunctionCall(Alpha=ComponentMask(None=@n15) {r=true}, Min=Add(A=@n16, B=@n18), Max=Add(A=@n16, B=@n17)) {materialFunction=&SmoothStep}.Result)
  n20 = ScalarParameter() {defaultValue=0.125, group=BurstAnim, parameterName=ChipMaskThicknessU}
  n21 = Multiply(A=@n17) {constB=-1}
  n22 = OneMinus(None=MaterialFunctionCall(Alpha=ComponentMask(None=@n15) {g=true}, Min=Add(A=@n20, B=@n21), Max=Add(A=@n20, B=@n17)) {materialFunction=&SmoothStep}.Result)
  n23 = OneMinus(None=Clamp(None=DotProduct(A=@n19, B=@n22)))
  n24 = StaticSwitchParameter(True=@n23, False=Constant() {r=1}) {parameterName=UseCenterMask}
  n25 = Constant()
  n26 = MaterialFunctionCall(Alpha=@n9.R, Min=@n25, Max=ScalarParameter() {defaultValue=0.5, group=CircuitRanges, parameterName=SSCircuitNodeThickness}) {materialFunction=&SmoothStep}
  n27 = ScalarParameter() {defaultValue=0.5, group=CircuitRanges, parameterName=SSStaticCircuitNodeMult}
  n28 = Multiply(
   A=MaterialFunctionCall(Alpha=@n9.G, Min=@n25, Max=ScalarParameter() {defaultValue=0.5, group=CircuitRanges, parameterName=SSStaticCircuitThickness}) {materialFunction=&SmoothStep}.Result
   B=ScalarParameter() {defaultValue=0.5, group=CircuitRanges, parameterName=SSStaticCircuitThicknessMult}
  )
  n29 = TextureSampleParameter2D(UVs=@n8) {parameterName=CircuitTexture, texture=&T_UI_Datachips_Circuits}
  n30 = Multiply(A=Time(), B=ScalarParameter() {defaultValue=0.25, parameterName=TimeMult}) {constB=0.2}
  n31 = Add(A=Sine(None=Multiply(A=@n30) {constB=2}))
  n32 = Add(A=Sine(None=@n30))
  n33 = Constant() {r=1}
  n34 = LinearInterpolate(
   A=ScalarParameter() {group=CircuitAnimRange, parameterName=SineRangeLow}
   B=ScalarParameter() {defaultValue=0.9, group=CircuitAnimRange, parameterName=SineRangeHigh}
   Alpha=LinearInterpolate(A=Multiply(A=@n31) {constB=0.5}, B=Multiply(A=@n32) {constB=0.5}, Alpha=Fmod(A=@n30, B=@n33))
  )
  n35 = ScalarParameter() {defaultValue=0.05, group=CircuitRanges, parameterName=SSMarginCircuitRange}
  n36 = Add(A=@n34, B=@n35)
  n37 = ScalarParameter() {defaultValue=0.1, group=CircuitRanges, parameterName=SSMarginCircuitEdge}
  n38 = Multiply(A=@n37) {constB=-1}
  n39 = OneMinus(None=MaterialFunctionCall(Alpha=@n29.B, Min=Add(A=@n36, B=@n38), Max=Add(A=@n36, B=@n37)) {materialFunction=&SmoothStep}.Result)
  n40 = Add(A=@n34, B=Multiply(A=@n35) {constB=-1})
  n41 = MaterialFunctionCall(Alpha=@n29.B, Min=Add(A=@n40, B=Multiply(A=@n37) {constB=-1}), Max=Add(A=@n40, B=@n37)) {materialFunction=&SmoothStep}
  n42 = LinearInterpolate(
   B=Clamp(None=DotProduct(A=@n39, B=@n41.Result))
   Alpha=MaterialFunctionCall(Alpha=@n9.G, Min=@n25, Max=ScalarParameter() {defaultValue=0.5, group=CircuitRanges, parameterName=SSMaxCircuitThickness}) {materialFunction=&SmoothStep}.Result
  )
  n43 = Multiply(
   A=@n1."X 100%, Y 100%"
   B=AppendVector(
    A=ScalarParameter() {defaultValue=0.5, group=Noise, parameterName=Noise_TileAB_U}
    B=ScalarParameter() {defaultValue=1, group=Noise, parameterName=Noise_TileAB_V}
   )
  )
  n44 = ScalarParameter() {defaultValue=-0.05, group=Noise, parameterName=Noise_SpeedU}
  n45 = ScalarParameter() {defaultValue=-0.1, group=Noise, parameterName=Noise_SpeedV}
  n46 = TextureSample(UVs=Panner(Coordinate=@n43, Speed=AppendVector(A=@n44, B=@n45)) {speedX=0.25}) {samplerType=SAMPLERTYPE_Grayscale, texture=&LowResBlurredNoise}
  n47 = ScalarParameter() {defaultValue=2, group=Noise, parameterName=NoiseMultA}
  n48 = ScalarParameter() {defaultValue=-0.05, group=Noise, parameterName=NoiseB_SpeedU}
  n49 = ScalarParameter() {defaultValue=-0.1, group=Noise, parameterName=NoiseB_SpeedV}
  n50 = TextureSample(UVs=Panner(Coordinate=@n43, Speed=AppendVector(A=@n48, B=@n49)) {speedX=0.25}) {samplerType=SAMPLERTYPE_Grayscale, texture=&LowResBlurredNoise}
  n51 = ScalarParameter() {defaultValue=2, group=Noise, parameterName=NoiseMultB}
  n52 = DotProduct(
   A=Power(Base=Multiply(A=@n46.RGB, B=@n47) {constB=4}, Exp=ScalarParameter() {defaultValue=2, group=Noise, parameterName=NoisePowerA}) {constExponent=16}
   B=Power(Base=Multiply(A=@n50.RGB, B=@n51) {constB=4}, Exp=ScalarParameter() {defaultValue=2, group=Noise, parameterName=NoisePowerB}) {constExponent=16}
  )
  n53 = TextureCoordinate()
  n54 = AppendVector(
   A=ScalarParameter() {defaultValue=0.5, group=Noise, parameterName=Noise_TileC_U}
   B=ScalarParameter() {defaultValue=1, group=Noise, parameterName=Noise_TileC_V}
  )
  n55 = ScalarParameter() {defaultValue=-0.05, group=Noise, parameterName=NoiseC_SpeedU}
  n56 = ScalarParameter() {defaultValue=-0.1, group=Noise, parameterName=NoiseC_SpeedV}
  n57 = TextureSample(UVs=Panner(
    Coordinate=MaterialFunctionCall(UVs=@n53, Texture Scale=@n54) {materialFunction=&ScaleUVsByCenter}.UVs
    Speed=AppendVector(A=@n55, B=@n56)
   ) {speedX=0.25}) {samplerType=SAMPLERTYPE_Grayscale, texture=&LowResBlurredNoise}
  n58 = ScalarParameter() {defaultValue=2, group=Noise, parameterName=NoiseMultC}
  n59 = DotProduct(
   A=@n52
   B=Power(Base=Multiply(A=@n57.RGB, B=@n58) {constB=4}, Exp=ScalarParameter() {defaultValue=2, group=Noise, parameterName=NoisePowerC}) {constExponent=16}
  )
  n60 = LinearInterpolate(
   B=Add(A=Multiply(A=@n26.Result, B=@n27), B=Add(A=@n28, B=@n42))
   Alpha=Clamp(None=StaticSwitchParameter(True=@n59, False=@n52) {parameterName=UseNoiseChannelC})
  )
  n61 = ScalarParameter() {group=BurstAnim, parameterName=BaseSineAlpha}
  n62 = TextureCoordinate()
  n63 = AppendVector(
   A=Multiply(A=ScalarParameter() {defaultValue=0.5, group=BurstAnim, parameterName=BurstPositionU}) {constB=-1}
   B=Multiply(A=ScalarParameter() {defaultValue=0.5, group=BurstAnim, parameterName=BurstPositionV}) {constB=-1}
  )
  n64 = Panner(
   Coordinate=MaterialFunctionCall(Vector or UVs=Add(A=@n62, B=@n63)) {materialFunction=&VectorToRadialValue}.Linear Distance
   Speed=ScalarParameter() {defaultValue=-0.25, group=BurstAnim, parameterName=SineSpeed}
  ) {speedX=-0.25}
  n65 = ComponentMask(None=Sine(None=TextureSample(UVs=@n64) {texture=&/T_UI_Datachips_Circuits_BurstGrad}.RGB) {period=4}) {r=true}
  n66 = LinearInterpolate(
   B=LinearInterpolate(B=LinearInterpolate(A=@n61, Alpha=@n65), Alpha=@n13.RadialGradientExponential)
   Alpha=StaticSwitchParameter(True=@n23, False=Constant() {r=1}) {parameterName=UseCenterMask}
  )
  StaticSwitchParameter(True=Add(A=LinearInterpolate(B=@n14, Alpha=@n24), B=LinearInterpolate(B=@n60, Alpha=@n66)), False=@n60) {parameterName=IsCircuitBurst}
 ]
}
