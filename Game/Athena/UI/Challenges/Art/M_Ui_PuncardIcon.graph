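# UI-domain translucent material for punch card challenge icons, tinting an ItemIcon texture with color parameters; the default texture is a combat pickaxe.
# One blend mixes Accent01 and Accent02 by the icon's red channel, then blends toward white by green.
# A second chain also blends in BaseColor and uses summed, saturated masks to fade from white, though its mask reroutes show no source.
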Material /Game/Athena/UI/Challenges/Art/M_Ui_PuncardIcon
graph {
 properties {
  bUsedWithStaticMesh=false
  blendMode=BLEND_Translucent
  materialDomain=MD_UI
  pixelDepthOffsetMode=PDOM_Legacy
  usedShadingModels=Default Lit
 }
 nodes [
  n1 = Reroute()
  n2 = VectorParameter() {parameterName=Accent01, defaultValue=<rgba 1 0.8571 0 1>}
  n3 = VectorParameter() {parameterName=Accent02, defaultValue=<rgba 0 0.632 1 1>}
  n4 = Reroute()
  n5 = Add(A=@n1, B=@n4)
  n6 = Saturate(None=Add(A=Reroute(), B=Saturate(None=@n5)))
  LinearInterpolate(
   A=Constant() {r=1}
   B=LinearInterpolate(
    A=LinearInterpolate(A=@n2.RGB, B=@n3.RGB, Alpha=@n4)
    B=VectorParameter() {parameterName=BaseColor, defaultValue=<rgba 0 0.3855 1 1>}.RGB
    Alpha=@n1
   )
   Alpha=Saturate(None=Multiply(A=@n6) {constB=2})
  )
  n7 = TextureSampleParameter2D() {
   parameterName=ItemIcon
   samplerSource=SSM_Clamp_WorldGroupSettings
   texture=/Game/Athena/UI/Challenges/Art/PunchCard_Icons/T_Ui_Combat_Pickaxe
  }
  LinearInterpolate(
   A=LinearInterpolate(
    A=VectorParameter() {parameterName=Accent01, defaultValue=<rgba 1 0.8571 0 1>}.RGB
    B=VectorParameter() {parameterName=Accent02, defaultValue=<rgba 0 0.632 1 1>}.RGB
    Alpha=@n7.R
   )
   B=Constant() {r=1}
   Alpha=@n7.G
  )
  Reroute(None=Reroute())
  Reroute(None=@n7.A)
  VectorParameter() {parameterName=BaseColor, defaultValue=<rgba 0 0.3855 1 1>}
 ]
}
