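# Translucent UI material that draws a challenge punchcard tile with a hover-reactive border and a clipped top-right corner.
# Rotated-UV masks build the border, whose angles, opacity and color blend on a hover parameter.
# The fill layers radial gradients, a center split, a category texture tiled by mixed screen and local UVs, and a shine swipe.
# An Is Unknown switch swaps in flat dark color, divet masking and altered alpha.
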Material /Game/Athena/UI/Challenges/Art/M_Ui_Punchcard
graph {
 properties {
  bUsedWithStaticMesh=false
  blendMode=BLEND_Translucent
  materialDomain=MD_UI
  pixelDepthOffsetMode=PDOM_Legacy
  usedShadingModels=Default Lit
 }
 nodes [
  Constant() {r=1}
  n1 = VectorParameter() {group=colors and alpha, parameterName=BaseColor03, defaultValue=<rgba 0.016 0.0666 0.4969 1>}
  n2 = VectorParameter() {group=colors and alpha, parameterName=BaseColor02, defaultValue=<rgba 0.016 0.1714 0.7835 1>}
  n3 = ScalarParameter() {defaultValue=0.5, group=colors and alpha, parameterName=bordercolor}
  n4 = VectorParameter() {group=colors and alpha, parameterName=BaseColor04, defaultValue=<rgba 0.0003 0.0284 0.1683 1>}
  n5 = Saturate(None=ComponentMask(None=TextureCoordinate() {coordinateIndex=4}) {g=true})
  LinearInterpolate(A=LinearInterpolate(A=@n1.RGB, B=@n2.RGB, Alpha=@n3), B=LinearInterpolate(A=@n4.RGB, B=@n1.RGB, Alpha=@n3), Alpha=@n5)
  n6 = Add(A=TextureCoordinate()) {constB=-0.5}
  n7 = Subtract(A=Multiply(A=Abs(None=@n6)) {constB=2}, B=ScalarParameter() {defaultValue=0.997, group=Outer_Edge, parameterName=Outeredge_AMOUNT}) {constB=0.5}
  n8 = ScalarParameter() {defaultValue=64, group=Outer_Edge, parameterName=OuterEdge_Sharpness}
  n9 = MaterialFunctionCall(Vector 2=Saturate(None=Multiply(A=@n7, B=@n8) {constB=128})) {materialFunction=/Engine/Functions/Engine_MaterialFunctions02/Utility/VectorLength}
  n10 = LinearInterpolate(
   A=Constant() {r=1}
   B=OneMinus(None=Multiply(A=@n9.V2 Length))
   Alpha=ScalarParameter() {defaultValue=1, group=Outer_Edge, parameterName=MakeOuterEdge}
  )
  n11 = Subtract(A=@n10, B=ScalarParameter() {defaultValue=0.75, parameterName=AlphaBg})
  n12 = TextureCoordinate() {coordinateIndex=4}
  n13 = LinearInterpolate(
   A=ScalarParameter() {defaultValue=0.002, group=Border, parameterName=Top-Angle}
   B=ScalarParameter() {defaultValue=-0.002, group=Border, parameterName=Top-Angle-hover}
   Alpha=ScalarParameter() {parameterName=hover, sliderMax=1}
  )
  n14 = Subtract(
   A=ComponentMask(None=Rotator(Coordinate=@n12, Time=@n13) {speed=6.2832}) {g=true}
   B=LinearInterpolate(
    A=ScalarParameter() {defaultValue=0.04, group=Border, parameterName=Top-RangeOffset, sliderMax=1}
    B=ScalarParameter() {defaultValue=0.04, group=Border, parameterName=Top-RangeOffset-hover, sliderMax=1}
    Alpha=ScalarParameter() {parameterName=hover, sliderMax=1}
   )
  )
  n15 = ScalarParameter() {defaultValue=512, group=Feather, parameterName=HorizontalSharpness}
  n16 = TextureCoordinate() {coordinateIndex=4}
  n17 = LinearInterpolate(
   A=ScalarParameter() {defaultValue=-0.002, group=Border, parameterName=Left-Angle}
   B=ScalarParameter() {defaultValue=0.001, group=Border, parameterName=Left-Angle-hover}
   Alpha=ScalarParameter() {parameterName=hover, sliderMax=1}
  )
  n18 = Subtract(
   A=ComponentMask(None=Rotator(Coordinate=@n16, Time=@n17) {speed=6.2832}) {r=true}
   B=LinearInterpolate(
    A=ScalarParameter() {defaultValue=0.02, group=Border, parameterName=Left-RangeOffset, sliderMax=1}
    B=ScalarParameter() {defaultValue=0.02, group=Border, parameterName=Left-RangeOffset-hover, sliderMax=1}
    Alpha=ScalarParameter() {parameterName=hover, sliderMax=1}
   )
  )
  n19 = ScalarParameter() {defaultValue=512, group=Feather, parameterName=VerticalSharpness}
  n20 = DotProduct(A=Saturate(None=Multiply(A=@n14, B=@n15)), B=Saturate(None=Multiply(A=@n18, B=@n19)))
  n21 = TextureCoordinate() {coordinateIndex=4}
  n22 = LinearInterpolate(
   A=ScalarParameter() {defaultValue=0.5025, group=Border, parameterName=Bot-Angle}
   B=ScalarParameter() {defaultValue=0.4975, group=Border, parameterName=Bot-Angle-hover}
   Alpha=ScalarParameter() {parameterName=hover, sliderMax=1}
  )
  n23 = Subtract(
   A=ComponentMask(None=Rotator(Coordinate=@n21, Time=@n22) {speed=6.2832}) {g=true}
   B=LinearInterpolate(
    A=ScalarParameter() {defaultValue=0.04, group=Border, parameterName=Bot-RangeOffset, sliderMax=1}
    B=ScalarParameter() {defaultValue=0.04, group=Border, parameterName=Bot-RangeOffset-hover, sliderMax=1}
    Alpha=ScalarParameter() {parameterName=hover, sliderMax=1}
   )
  )
  n24 = ScalarParameter() {defaultValue=512, group=Feather, parameterName=HorizontalSharpness}
  n25 = TextureCoordinate() {coordinateIndex=4}
  n26 = LinearInterpolate(
   A=ScalarParameter() {defaultValue=0.4975, group=Border, parameterName=Right-Angle}
   B=ScalarParameter() {defaultValue=0.5015, group=Border, parameterName=Right-Angle-hover}
   Alpha=ScalarParameter() {parameterName=hover, sliderMax=1}
  )
  n27 = Subtract(
   A=ComponentMask(None=Rotator(Coordinate=@n25, Time=@n26) {speed=6.2832}) {r=true}
   B=LinearInterpolate(
    A=ScalarParameter() {defaultValue=0.02, group=Border, parameterName=Right-RangeOffset, sliderMax=1}
    B=ScalarParameter() {defaultValue=0.02, group=Border, parameterName=Right-RangeOffset-hover, sliderMax=1}
    Alpha=ScalarParameter() {parameterName=hover, sliderMax=1}
   )
  )
  n28 = ScalarParameter() {defaultValue=512, group=Feather, parameterName=VerticalSharpness}
  n29 = DotProduct(A=Saturate(None=Multiply(A=@n23, B=@n24)), B=Saturate(None=Multiply(A=@n27, B=@n28)))
  n30 = Rotator(
   Coordinate=TextureCoordinate() {coordinateIndex=4}
   Time=LinearInterpolate(
    A=ScalarParameter() {defaultValue=0.58, group=Border, parameterName=TopRight-Angle}
    B=ScalarParameter() {defaultValue=0.58, group=Border, parameterName=TopRight-Angle-hover}
    Alpha=ScalarParameter() {parameterName=hover, sliderMax=1}
   )
  ) {speed=6.2832}
  n31 = ScalarParameter() {defaultValue=-0.125, group=Border, parameterName=TopRight-RangeOffset}
  n32 = ScalarParameter() {defaultValue=-0.125, group=Border, parameterName=TopRight-RangeOffset-hover}
  n33 = ScalarParameter() {parameterName=hover, sliderMax=1}
  n34 = Multiply(
   A=Subtract(A=ComponentMask(None=@n30) {r=true}, B=LinearInterpolate(A=@n31, B=@n32, Alpha=@n33))
   B=ScalarParameter() {defaultValue=512, group=Feather, parameterName=HorizontalSharpness}
  )
  n35 = Saturate(None=DotProduct(A=DotProduct(A=@n20, B=@n29), B=Saturate(None=@n34)))
  n36 = TextureCoordinate() {coordinateIndex=4}
  n37 = LinearInterpolate(
   A=ScalarParameter() {defaultValue=0.002, group=Divet, parameterName=Divet-Top-Angle}
   B=ScalarParameter() {defaultValue=-0.002, group=Divet, parameterName=Divet-Top-Angle-hover}
   Alpha=ScalarParameter() {parameterName=hover, sliderMax=1}
  )
  n38 = Subtract(
   A=ComponentMask(None=Rotator(Coordinate=@n36, Time=@n37) {speed=6.2832}) {g=true}
   B=LinearInterpolate(
    A=ScalarParameter() {defaultValue=0.04, group=Divet, parameterName=Divet-Top-RangeOffset, sliderMax=1}
    B=ScalarParameter() {defaultValue=0.04, group=Divet, parameterName=DIvet-Top-RangeOffset-hover, sliderMax=1}
    Alpha=ScalarParameter() {parameterName=hover, sliderMax=1}
   )
  )
  n39 = ScalarParameter() {defaultValue=512, group=Feather, parameterName=HorizontalSharpness}
  n40 = OneMinus(None=Saturate(None=Multiply(A=@n38, B=@n39)))
  n41 = TextureCoordinate() {coordinateIndex=4}
  n42 = LinearInterpolate(
   A=ScalarParameter() {defaultValue=-0.002, group=Divet, parameterName=DIvet-Left-Angle}
   B=ScalarParameter() {defaultValue=0.001, group=Divet, parameterName=Divet-Left-Angle-hover}
   Alpha=ScalarParameter() {parameterName=hover, sliderMax=1}
  )
  n43 = Subtract(
   A=ComponentMask(None=Rotator(Coordinate=@n41, Time=@n42) {speed=6.2832}) {r=true}
   B=LinearInterpolate(
    A=ScalarParameter() {defaultValue=0.02, group=Divet, parameterName=Divet-Left-RangeOffset, sliderMax=1}
    B=ScalarParameter() {defaultValue=0.02, group=Divet, parameterName=Divet-Left-RangeOffset-hover, sliderMax=1}
    Alpha=ScalarParameter() {parameterName=hover, sliderMax=1}
   )
  )
  n44 = ScalarParameter() {defaultValue=512, group=Feather, parameterName=VerticalSharpness}
  n45 = OneMinus(None=Saturate(None=Multiply(A=@n43, B=@n44)))
  n46 = Add(
   A=OneMinus(None=@n35)
   B=StaticSwitchParameter(True=DotProduct(A=@n40, B=@n45), False=Constant()) {group=Is Unknown, parameterName=Is Unknown}
  )
  n47 = ScalarParameter() {defaultValue=0.5, group=colors and alpha, parameterName=BorderAlpha, sliderMax=1}
  n48 = ScalarParameter() {defaultValue=1, group=colors and alpha, parameterName=HoveredBorderAlpha, sliderMax=1, sliderMin=0.01}
  n49 = ScalarParameter() {parameterName=hover, sliderMax=1}
  n50 = Saturate(None=Multiply(A=Saturate(None=@n46), B=LinearInterpolate(A=@n47, B=@n48, Alpha=@n49) {constB=0}))
  n51 = Saturate(None=Add(A=@n50, B=Saturate(None=@n11)))
  n52 = TextureCoordinate() {coordinateIndex=4}
  n53 = Constant() {r=0.58}
  n54 = Subtract(A=ComponentMask(None=Rotator(Coordinate=@n52, Time=@n53) {speed=6.2832}) {r=true}, B=Constant() {r=-0.125})
  n55 = ScalarParameter() {defaultValue=512, group=Feather, parameterName=HorizontalSharpness}
  Multiply(
   A=StaticSwitchParameter(True=Multiply(A=@n51, B=@n10), False=@n10) {group=Is Unknown, parameterName=Is Unknown}
   B=Saturate(None=Multiply(A=@n54, B=@n55) {constB=128})
  )
  n56 = MaterialFunctionCall(
   UVs=TextureCoordinate() {coordinateIndex=4}
   CenterPosition=AppendVector(A=Constant() {r=0.5}, B=ScalarParameter() {defaultValue=0.33, group=Radial Gradient, parameterName=RadailGradient Position})
   Radius=ScalarParameter() {defaultValue=0.75, group=Radial Gradient, parameterName=Radial Gradient Radius}
   Density=ScalarParameter() {defaultValue=1, group=Radial Gradient, parameterName=Radial Gradient Density}
  ) {materialFunction=&RadialGradientExponential}
  n57 = LinearInterpolate(A=@n4.RGB, B=@n1.RGB, Alpha=@n56.RadialGradientExponential)
  n58 = LinearInterpolate(A=@n1.RGB, B=@n2.RGB, Alpha=@n56.RadialGradientExponential)
  n59 = TextureCoordinate() {coordinateIndex=4}
  n60 = Divide(A=ScalarParameter() {group=Degrees, parameterName=CenterAngle}) {constB=360}
  n61 = ComponentMask(None=Add(
    A=Rotator(Coordinate=@n59, Time=@n60) {centerX=0, speed=6.2832}
    B=ScalarParameter() {defaultValue=0.5, group=Offset, parameterName=CenterOffset}
   )) {g=true}
  n62 = Saturate(None=Multiply(A=OneMinus(None=@n61), B=ScalarParameter() {defaultValue=512, group=Feather, parameterName=HorizontalSharpness}))
  n63 = Saturate(None=Subtract(A=@n62, B=ScalarParameter() {defaultValue=0.5, parameterName=FadeUpperPart, sliderMax=1}) {constB=0.5})
  n64 = MaterialFunctionCall(Float2=MaterialFunctionCall() {materialFunction=&ScreenResolution}.Visible Resolution) {materialFunction=&BreakOutFloat2Components}
  n65 = Divide(A=@n64.R, B=Max(A=@n64.G))
  n66 = Constant() {r=1}
  n67 = Multiply(
   A=Multiply(
    A=AppendVector(A=@n65, B=@n66)
    B=MaterialFunctionCall() {materialFunction=&ScreenAlignedUVs}."X 100%, Y 100%"
   )
   B=ScalarParameter() {defaultValue=4, parameterName=TileScreenSpace}
  )
  n68 = MaterialFunctionCall() {materialFunction=&GetUserInterfaceUV}
  n69 = MaterialFunctionCall(Float2=@n68.Pixel Size) {materialFunction=&BreakOutFloat2Components}
  n70 = Divide(A=@n69.R, B=Max(A=@n69.G))
  n71 = Constant() {r=1}
  n72 = Multiply(A=Multiply(A=AppendVector(A=@n70, B=@n71), B=@n68.Normalized UV), B=ScalarParameter() {defaultValue=0.25, parameterName=Local UV})
  n73 = TextureObjectParameter() {group=background, parameterName=Background, texture=/Game/Athena/UI/Challenges/Art/PunchCard_Icons/CategoriesTextures/T_UI_punchcard_Combat}
  n74 = TextureSample(UVs=Multiply(A=Add(A=@n67, B=@n72), B=ScalarParameter() {defaultValue=0.5, group=background, parameterName=TileAmount}), Tex=@n73) {texture=&T_UI_punchcard_Combat}
  n75 = OneMinus(None=@n56.RadialGradientExponential)
  n76 = ScalarParameter() {defaultValue=1, group=background, parameterName=unhovered gradient mask}
  n77 = LinearInterpolate(
   A=Saturate(None=Add(A=@n75, B=@n76))
   B=OneMinus(None=@n56.RadialGradientExponential)
   Alpha=ScalarParameter() {parameterName=hover, sliderMax=1}
  )
  n78 = Multiply(A=@n74.G, B=@n77)
  n79 = Saturate(None=Add(A=@n62, B=ScalarParameter() {defaultValue=0.5, group=background, parameterName=Bottom_alpha}) {constB=0.5})
  n80 = Saturate(None=Multiply(A=Saturate(None=@n78), B=@n79))
  n81 = VectorParameter() {group=colors and alpha, parameterName=BaseColor03, defaultValue=<rgba 0.016 0.0666 0.4969 1>}
  n82 = VectorParameter() {group=colors and alpha, parameterName=BaseColor02, defaultValue=<rgba 0.016 0.1714 0.7835 1>}
  n83 = Multiply(A=Saturate(None=Multiply(A=@n74.R, B=@n77)), B=@n79)
  n84 = LinearInterpolate(
   A=LinearInterpolate(A=LinearInterpolate(A=@n57, B=@n58, Alpha=@n63), B=LinearInterpolate(A=@n81.RGB, B=@n82.RGB, Alpha=@n62), Alpha=Multiply(A=@n80, B=@n35))
   B=@n4.RGB
   Alpha=Multiply(A=Saturate(None=@n83), B=@n35)
  )
  n85 = Saturate(None=Multiply(A=@n74.G, B=Constant() {r=2}))
  n86 = MaterialFunctionCall!(In_Uv=MaterialFunctionCall() {materialFunction=&ScreenAlignedUVs}."X 100%, Y 100%") {materialFunction=/Game/UI/Foundation/Materials/M-Function-ShineSwipe}
  n87 = Multiply(A=Multiply(A=@n62, B=Multiply(A=@n85, B=@n86.Result)), B=@n77)
  n88 = LinearInterpolate(A=VectorParameter() {parameterName=BaseColor0Hightower, defaultValue=<rgba 1 0.2565 0 0>}.RGB, B=@n2.RGB, Alpha=@n5)
  n89 = Constant() {r=1}
  n90 = ScalarParameter() {parameterName=hover, sliderMax=1}
  StaticSwitchParameter(
   True=VectorParameter() {group=colors and alpha, parameterName=BaseColor04, defaultValue=<rgba 0.0003 0.0284 0.1683 1>}.RGB
   False=LinearInterpolate(A=LinearInterpolate(A=@n84, B=@n2.RGB, Alpha=@n87), B=LinearInterpolate(A=@n88, B=@n89, Alpha=@n90) {constB=0}, Alpha=@n50)
  ) {group=Is Unknown, parameterName=Is Unknown}
 ]
}
