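# Translucent UI material for a challenge punchcard slot, drawing a procedural dot with optional horizontal connector bars on either side.
# A sharpened exponential radial gradient forms the circle, with adjustable radius and edge sharpness.
# Hard-edged thresholds on UV channel 4 cut a centered band and split it at the midpoint into halves faded by ShowLeft and ShowRight.
# Circle and halves are summed and saturated into one mask, alongside a white color constant.
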Material /Game/Athena/UI/Challenges/Art/M_Ui_Punchcard_slot
graph {
 properties {
  bUsedWithStaticMesh=false
  blendMode=BLEND_Translucent
  materialDomain=MD_UI
  pixelDepthOffsetMode=PDOM_Legacy
  usedShadingModels=Default Lit
 }
 nodes [
  Constant3Vector() {constant=<rgba 1 1 1 0>}
  n1 = Multiply(
   A=MaterialFunctionCall(Radius=ScalarParameter() {defaultValue=0.45, parameterName=Radius}) {materialFunction=&RadialGradientExponential}.RadialGradientExponential
   B=ScalarParameter() {defaultValue=512, parameterName=CircleShapness}
  )
  n2 = ScalarParameter() {defaultValue=0.1, parameterName=BarThickness, sliderMax=0.5}
  n3 = Add(B=@n2) {constA=-0.5}
  n4 = TextureCoordinate() {coordinateIndex=4}
  n5 = Saturate(None=ComponentMask(None=Add(A=@n4, B=@n3) {constB=-0.5}) {g=true})
  n6 = ScalarParameter() {defaultValue=512, parameterName=BarShapness}
  n7 = Add(A=@n4, B=Subtract(B=@n2) {constA=-0.5}) {constB=-0.5}
  n8 = Multiply(A=Saturate(None=ComponentMask(None=@n7) {g=true}), B=@n6)
  n9 = Saturate(None=Multiply(A=Multiply(A=@n5, B=@n6), B=OneMinus(None=@n8)))
  n10 = ComponentMask(None=Add(A=@n4) {constB=-0.5}) {r=true}
  n11 = Saturate(None=Multiply(A=Saturate(None=@n10), B=@n6))
  n12 = LinearInterpolate(A=Constant(), B=Multiply(A=@n9, B=@n11), Alpha=ScalarParameter() {defaultValue=1, parameterName=ShowRight})
  n13 = LinearInterpolate(A=Constant(), B=Multiply(A=@n9, B=OneMinus(None=@n11)), Alpha=ScalarParameter() {defaultValue=1, parameterName=ShowLeft})
  Saturate(None=Add(A=Saturate(None=@n1), B=Add(A=@n12, B=@n13)))
 ]
}
